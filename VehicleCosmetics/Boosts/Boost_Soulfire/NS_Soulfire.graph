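# NS_Soulfire is the Niagara effect for the Soulfire car boost cosmetic, trailing ribbons from the vehicle's boost ports.
# One emitter spawns ribbons per port at skeletal positions, with game mode selecting lifetime and backward velocity.
# Painted-variant curves color three layers (R1–R3), while curves shape R1 and R2 widths across two ribbon renderers.
# An R1 position offset adds age-phased vertical waves, with amplitude adjusted by the camera-to-car angle.
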NiagaraSystem /VehicleCosmetics/Boosts/Boost_Soulfire/NS_Soulfire
dependencies {
 usedDataInterfaces [
  &NiagaraDataInterfaceColorCurve
  &NiagaraDataInterfaceCurve
  &NiagaraDataInterfaceCamera
  &NiagaraDataInterfaceSkeletalMesh
 ]
 usedDynamicInputs []
 usedModules [
  &self:SystemSpawnScript.SystemScriptSource.SystemScriptGraph.NiagaraNodeAssignment_0.SetVariables_#1
  &SystemState
 ]
}
emitters {NE_VC_Boost_Ribbon_Parent001 {
  emitterUpdateScript [
   &EmitterState(Life Cycle Mode=::NewEnumerator1, Loop Duration=1, UseLoopDelay=false)
   &VC_FilteredSocketCount()
   &VC_RateQualityHandler(Epic Spawn Rate=30, High Spawn Rate=30, Low Spawn Rate=30, Use Custom Rate?=true)
   &VC_SpawnRibbon_v1(Number of Ribbons=$Emitter.PortCount, SpawnRate=$Emitter.Spawn Rate)
  ]
  particleSpawnScript [
   &VC_PaintedVariantsOverride(
    DefaultColor=&ColorFromCurve(ColorCurve=DataInterface() {propertyValues {
       AlphaCurve {keys [<at 0 0>,<at 0.0745 1 RCIM_Cubic>,<at 0.8087 0 RCIM_Cubic>]}
       BlueCurve {keys [<at 0 1.6 RCIM_Cubic>,<at 0.3 0.6 RCIM_Cubic>]}
       GreenCurve {keys [<at 0 4 RCIM_Cubic>,<at 0.3 2 RCIM_Cubic>]}
       RedCurve {keys [<at 0 2 RCIM_Cubic>,<at 0.3 0 RCIM_Cubic>]}
      }})
    R1 DefaultColor=<rgba 0 1 0.5972 0.7765>
    R2 DefaultColor:LinearColor=<rgba 1 0 0.8 1>
    Use R1 Painted Overrides=true
    Use R2 Painted Overrides=true
   ) {enabled=false}
   &VC_InitializeRibbon_v1(
    Color=&ColorFromCurve(ColorCurve=DataInterface() {propertyValues {
       AlphaCurve {keys [<at 0 1.5 RCIM_Cubic>]}
       BlueCurve {keys [<at 0 0.5476>, {interpMode=RCIM_Cubic, leaveTangent=-0.2097, time=0.6374, value=0.4598}]}
       GreenCurve {keys [<at 0 0.8958>, {interpMode=RCIM_Cubic, leaveTangent=-0.4942, time=0.6374, value=0.5781}]}
       RedCurve {keys [<at 0 0.4199>,<at 0.6374 0.0241>]}
      }})
    Custom Offset=true
    Enable Skeletal Position=true
    Facing Axis:ENiagaraOrientationAxis=::YAxis
    Lifetime=&MergeHelper_FloatSwitch(0=1, 1=0.5, 2=0.5, 3=?, Index=$System.GameModeID, Num=::NewEnumerator1)
    Local Vector Offset=&AddVectorToPosition(Position=<-12 0 0>, Vector=Linked() {linkedVariable {name=User.DTCarOffset002, type=NiagaraPosition}})
    Offset Amount=1
    Ribbon Width=90
   )
   &AddVelocity(Velocity=&MergeHelper_VectorSwitch(0=<-1200 0 0>, 1=<-1300 0 0>, 2=<-1100 0 0>, 3=?, Index=$System.GameModeID, Num=::NewEnumerator1)) {module=AddVelocity001}
  ]
  particleUpdateScript [
   &VC_PaintedVariantsOverride(
    Orange=<rgba 0.6724 0.0883 0.2836 1>
    R1 Black:LinearColor=&ColorFromCurve(ColorCurve=DataInterface() {propertyValues {
       AlphaCurve {keys [<at 0 1.8 RCIM_Cubic>]}
       BlueCurve {keys [<at 0.2355 0.1042>,<at 0.6434 0.0417>]}
       GreenCurve {keys [<at 0.2355 0.1042>,<at 0.6434 0.0417>]}
       RedCurve {keys [<at 0.2355 0.1042>,<at 0.6434 0.0417>]}
      }})
    R1 BurntSienna:LinearColor=&ColorFromCurve(ColorCurve=DataInterface() {propertyValues {
       AlphaCurve {keys [<at 0 1.6 RCIM_Cubic>]}
       BlueCurve {keys [<at 0 0.1867>,<at 0.4772 0.1273>]}
       GreenCurve {keys [<at 0 0.2296>,<at 0.4772 0.1515>]}
       RedCurve {keys [<at 0 0.3854>,<at 0.4772 0.2396>]}
      }})
    R1 Cobalt:LinearColor=&ColorFromCurve(ColorCurve=DataInterface() {propertyValues {
       AlphaCurve {keys [<at 0 1.5 RCIM_Cubic>]}
       BlueCurve {keys [<at 0 0.8385 RCIM_Cubic>,<at 0.4999 0.6146>]}
       GreenCurve {keys [<at 0 0.2841 RCIM_Cubic>,<at 0.4999 0.1385>]}
       RedCurve {keys [<at 0 0.1092 RCIM_Cubic>,<at 0.4999 0.0224>]}
      }})
    R1 Crimson:LinearColor=&ColorFromCurve(ColorCurve=DataInterface() {propertyValues {
       AlphaCurve {keys [<at 0 2 RCIM_Cubic>]}
       BlueCurve {keys [<at 0 0.1628>,<at 0.515 0.0561>]}
       GreenCurve {keys [<at 0 0.1381>,<at 0.515 0.0601>]}
       RedCurve {keys [<at 0 0.6979>,<at 0.515 0.8281>]}
      }})
    R1 DefaultColor=&ColorFromCurve(ColorCurve=DataInterface() {propertyValues {
       AlphaCurve {keys [<at 0 1 RCIM_Cubic>]}
       BlueCurve {keys [<at 0.2135 0.3523>, {interpMode=RCIM_Cubic, leaveTangent=-0.2097, time=0.3838, value=0.3264}]}
       GreenCurve {keys [<at 0.2135 0.6615>, {interpMode=RCIM_Cubic, leaveTangent=-0.4942, time=0.3838, value=0.5781}]}
       RedCurve {keys [<at 0.2135 0.2102>,<at 0.3838 0.001>]}
      }})
    R1 ForestGreen:LinearColor=&ColorFromCurve(ColorCurve=DataInterface() {propertyValues {
       AlphaCurve {keys [<at 0 1.8 RCIM_Cubic>]}
       BlueCurve {keys [<at 0.0075 0.3253>,<at 0.5225 0.2786>]}
       GreenCurve {keys [<at 0.0075 0.6771>,<at 0.5225 0.5625>]}
       RedCurve {keys [<at 0.0075 0.3068>,<at 0.5225 0.2637>]}
      }})
    R1 Grey:LinearColor=&ColorFromCurve(ColorCurve=DataInterface() {propertyValues {
       AlphaCurve {keys [<at 0 1.7 RCIM_Cubic>]}
       BlueCurve {keys [<at 0 0.3229>,<at 0.5502 0.2604>]}
       GreenCurve {keys [<at 0 0.3229>,<at 0.5502 0.2604>]}
       RedCurve {keys [<at 0 0.3229>,<at 0.5502 0.2604>]}
      }})
    R1 Lime:LinearColor=&ColorFromCurve(ColorCurve=DataInterface() {propertyValues {
       AlphaCurve {keys [<at 0 1.7 RCIM_Cubic>]}
       BlueCurve {keys [<at 0.014 0.275>,<at 0.588 0.2697>]}
       GreenCurve {keys [<at 0.014 0.5677>,<at 0.588 0.5885>]}
       RedCurve {keys [<at 0.014 0.4198>,<at 0.588 0.4274>]}
      }})
    R1 Orange:LinearColor=&ColorFromCurve(ColorCurve=DataInterface() {propertyValues {
       AlphaCurve {keys [<at 0 1.6 RCIM_Cubic>]}
       BlueCurve {keys [<at 0 0.0729 RCIM_Cubic>,<at 0.6217 0.0194>]}
       GreenCurve {keys [<at 0 0.3963 RCIM_Cubic>,<at 0.6217 0.2724>]}
       RedCurve {keys [<at 0 1 RCIM_Cubic>,<at 0.6217 0.7448>]}
      }})
    R1 Pink:LinearColor=&ColorFromCurve(ColorCurve=DataInterface() {propertyValues {
       AlphaCurve {keys [<at 0 1.6 RCIM_Cubic>]}
       BlueCurve {keys [<at 0 0.7665>,<at 0.6236 0.4994>]}
       GreenCurve {keys [<at 0 0.3594>,<at 0.6236 0.0987>]}
       RedCurve {keys [<at 0 1>,<at 0.6236 0.7292>]}
      }})
    R1 Purple:LinearColor=&ColorFromCurve(ColorCurve=DataInterface() {propertyValues {
       AlphaCurve {keys [<at 0 1.7 RCIM_Cubic>]}
       BlueCurve {keys [<at 0 0.8906>,<at 0.4699 0.8906>]}
       GreenCurve {keys [<at 0 0.0603>,<at 0.4699 0>]}
       RedCurve {keys [<at 0 0.3105>,<at 0.4699 0.2683>]}
      }})
    R1 Saffron:LinearColor=&ColorFromCurve(ColorCurve=DataInterface() {propertyValues {
       AlphaCurve {keys [<at 0 1.5 RCIM_Cubic>]}
       BlueCurve {keys [<at 0.0075 0.2312>,<at 0.5931 0.2312>]}
       GreenCurve {keys [<at 0.0075 0.5138>,<at 0.5931 0.5138>]}
       RedCurve {keys [<at 0.0075 0.6>,<at 0.5931 0.6>]}
      }})
    R1 SkyBlue:LinearColor=&ColorFromCurve(ColorCurve=DataInterface() {propertyValues {
       AlphaCurve {keys [<at 0 1.7 RCIM_Cubic>]}
       BlueCurve {keys [<at 0 1>,<at 0.5779 0.5521>]}
       GreenCurve {keys [<at 0 0.7495>,<at 0.5779 0.4138>]}
       RedCurve {keys [<at 0 0.2448>,<at 0.5779 0.1351>]}
      }})
    R1 TitaniumWhite:LinearColor=&ColorFromCurve(ColorCurve=DataInterface() {propertyValues {
       AlphaCurve {keys [<at 0 1.5 RCIM_Cubic>]}
       BlueCurve {keys [<at 0 0.6>,<at 0.8219 0.4219>]}
       GreenCurve {keys [<at 0 0.6>,<at 0.8219 0.4219>]}
       RedCurve {keys [<at 0 0.6>,<at 0.8219 0.4219>]}
      }})
    R2 Black:LinearColor=&ColorFromCurve(ColorCurve=DataInterface() {propertyValues {
       AlphaCurve {keys [{arriveTangent=0.2287, interpMode=RCIM_Cubic, leaveTangent=0.2287, tangentMode=RCTM_User, time=0, value=2.5}]}
       BlueCurve {keys [<at 0 0.1198>,<at 0.5704 0.0411>]}
       GreenCurve {keys [<at 0 0.1198>,<at 0.5704 0.0411>]}
       RedCurve {keys [<at 0 0.1198>,<at 0.5704 0.0411>]}
      }})
    R2 BurntSienna:LinearColor=&ColorFromCurve(ColorCurve=DataInterface() {propertyValues {
       AlphaCurve {keys [{arriveTangent=0.2287, interpMode=RCIM_Cubic, leaveTangent=0.2287, tangentMode=RCTM_User, time=0, value=2.5}]}
       BlueCurve {keys [<at 0 0.1254>,<at 0.6434 0.059>]}
       GreenCurve {keys [<at 0 0.1767>,<at 0.6434 0.0792>]}
       RedCurve {keys [<at 0 0.3594>,<at 0.6434 0.151>]}
      }})
    R2 Cobalt:LinearColor=&ColorFromCurve(ColorCurve=DataInterface() {propertyValues {
       AlphaCurve {keys [<at 0 2.5>]}
       BlueCurve {keys [<at 0 0.9688>,<at 0.6182 0.6771>]}
       GreenCurve {keys [<at 0 0.1989>,<at 0.6182 0.0529>]}
       RedCurve {keys [<at 0 0.1211>,<at 0.6182 0.0529>]}
      }})
    R2 Crimson:LinearColor=&ColorFromCurve(ColorCurve=DataInterface() {propertyValues {
       AlphaCurve {keys [{arriveTangent=6.056, interpMode=RCIM_Cubic, time=0, value=3}]}
       BlueCurve {keys [<at 0 0.237 RCIM_Cubic>,<at 0.6585 0.0228>]}
       GreenCurve {keys [<at 0 0.22 RCIM_Cubic>,<at 0.6585 0.0019>]}
       RedCurve {keys [<at 0 0.7969 RCIM_Cubic>,<at 0.6585 0.7888>]}
      }})
    R2 DefaultColor:LinearColor=&ColorFromCurve(ColorCurve=DataInterface() {propertyValues {
       AlphaCurve {keys [{arriveTangent=0.2287, interpMode=RCIM_Cubic, leaveTangent=0.2287, tangentMode=RCTM_User, time=0, value=2}]}
       BlueCurve {keys [<at 0 1>, <at 0.1157 1>, {interpMode=RCIM_Cubic, leaveTangent=-1.2582, time=0.3233, value=0.9062}, <at 0.5797 0.4965>]}
       GreenCurve {keys [<at 0 0>, <at 0.1157 0.0312>, {interpMode=RCIM_Cubic, leaveTangent=1.7831, time=0.3233, value=0.6117}, <at 0.5797 0.7448>]}
       RedCurve {keys [<at 0 0.5839>, <at 0.1157 0.3455>, {interpMode=RCIM_Cubic, leaveTangent=-0.8634, time=0.3233, value=0.1699}, <at 0.5797 0>]}
      }})
    R2 ForestGreen:LinearColor=&ColorFromCurve(ColorCurve=DataInterface() {propertyValues {
       AlphaCurve {keys [{arriveTangent=6.0856, interpMode=RCIM_Cubic, time=0, value=3}]}
       BlueCurve {keys [<at 0.1372 0.3365>,<at 0.6988 0.0574>]}
       GreenCurve {keys [<at 0.1372 0.8906>,<at 0.6988 0.5781>]}
       RedCurve {keys [<at 0.1372 0.2969>,<at 0.6988 0>]}
      }})
    R2 Grey:LinearColor=&ColorFromCurve(ColorCurve=DataInterface() {propertyValues {
       AlphaCurve {keys [{arriveTangent=7.4364, interpMode=RCIM_Cubic, time=0, value=3}]}
       BlueCurve {keys [<at 0 0.3385 RCIM_Cubic>,<at 0.724 0.1562>]}
       GreenCurve {keys [<at 0 0.3385 RCIM_Cubic>,<at 0.724 0.1562>]}
       RedCurve {keys [<at 0 0.3385 RCIM_Cubic>,<at 0.724 0.1562>]}
      }})
    R2 Lime:LinearColor=&ColorFromCurve(ColorCurve=DataInterface() {propertyValues {
       AlphaCurve {keys [{arriveTangent=6.9779, interpMode=RCIM_Cubic, time=0, value=3}]}
       BlueCurve {keys [<at 0 0.2159>,<at 0.583 0>]}
       GreenCurve {keys [<at 0 1.0629>,<at 0.583 0.9062>]}
       RedCurve {keys [<at 0 0.6406>,<at 0.583 0.4544>]}
      }})
    R2 Orange:LinearColor=&ColorFromCurve(ColorCurve=DataInterface() {propertyValues {
       AlphaCurve {keys [{arriveTangent=6.9884, interpMode=RCIM_Cubic, time=0, value=2.5}]}
       BlueCurve {keys [<at 0 0.125>, {interpMode=RCIM_Cubic, leaveTangent=-0.1666, time=0.5477, value=0}]}
       GreenCurve {keys [<at 0 0.4302>, {interpMode=RCIM_Cubic, leaveTangent=-0.3301, time=0.5477, value=0.2101}]}
       RedCurve {keys [<at 0 1>, {interpMode=RCIM_Cubic, leaveTangent=-0.1817, time=0.5477, value=0.6719}]}
      }})
    R2 Pink:LinearColor=&ColorFromCurve(ColorCurve=DataInterface() {propertyValues {
       AlphaCurve {keys [{arriveTangent=4.1431, interpMode=RCIM_Cubic, time=0, value=1.5}]}
       BlueCurve {keys [<at 0 1.328>,<at 0.5502 0.474 RCIM_Cubic>]}
       GreenCurve {keys [<at 0 0.1562>,<at 0.5502 0 RCIM_Cubic>]}
       RedCurve {keys [<at 0 2>,<at 0.5502 0.7448 RCIM_Cubic>]}
      }})
    R2 Purple:LinearColor=&ColorFromCurve(ColorCurve=DataInterface() {propertyValues {
       AlphaCurve {keys [{arriveTangent=5.2581, interpMode=RCIM_Cubic, time=0, value=1.5}]}
       BlueCurve {keys [<at 0.0048 1.311>,<at 0.5779 0.7031 RCIM_Cubic>]}
       GreenCurve {keys [<at 0.0048 0.0751>,<at 0.5779 0.0006 RCIM_Cubic>]}
       RedCurve {keys [<at 0.0048 0.3243>,<at 0.5779 0.1673 RCIM_Cubic>]}
      }})
    R2 Saffron:LinearColor=&ColorFromCurve(ColorCurve=DataInterface() {propertyValues {
       AlphaCurve {keys [<at 0.0068 1.5>]}
       BlueCurve {keys [<at 0 0.2375>,<at 0.583 0.0727>]}
       GreenCurve {keys [<at 0 0.9749>,<at 0.583 0.5796>]}
       RedCurve {keys [<at 0 1.2>,<at 0.583 0.7344>]}
      }})
    R2 SkyBlue:LinearColor=&ColorFromCurve(ColorCurve=DataInterface() {propertyValues {
       AlphaCurve {keys [<at 0 2.5>]}
       BlueCurve {keys [<at 0 1.3438>,<at 0.6258 0.88 RCIM_Cubic>]}
       GreenCurve {keys [<at 0 0.9793>,<at 0.6258 0.5931 RCIM_Cubic>]}
       RedCurve {keys [<at 0 0.245>,<at 0.6258 0.015 RCIM_Cubic>]}
      }})
    R2 TitaniumWhite:LinearColor=&ColorFromCurve(ColorCurve=DataInterface() {propertyValues {
       AlphaCurve {keys [<at 0 1.4>]}
       BlueCurve {keys [<at 0 0.9>,<at 0.5528 0.4115 RCIM_Cubic>]}
       GreenCurve {keys [<at 0 0.9>,<at 0.5528 0.4115 RCIM_Cubic>]}
       RedCurve {keys [<at 0 0.9>,<at 0.5528 0.4115 RCIM_Cubic>]}
      }})
    R3 Black:LinearColor=<rgba 0.0052 0.0052 0.0052 1>
    R3 BurntSienna:LinearColor=<rgba 0.0312 0.0096 0.0039 1>
    R3 Cobalt:LinearColor=<rgba 0 0.004 0.2135 1>
    R3 Crimson:LinearColor=<rgba 0.1458 0.0114 0.015 1>
    R3 DefaultAlpha=4
    R3 DefaultColor:LinearColor=<rgba 0.0011 0.0034 0.0208 1>
    R3 ForestGreen:LinearColor=<rgba 0 0.0781 0.0039 1>
    R3 Grey:LinearColor=<rgba 0.1042 0.1042 0.1042 1>
    R3 Lime:LinearColor=<rgba 0.0677 0.1354 0 1>
    R3 Orange:LinearColor=<rgba 0.3177 0.1108 0 1>
    R3 Pink:LinearColor=<rgba 0.2344 0.0281 0.1594 1>
    R3 Purple:LinearColor=<rgba 0.062 0 0.151 1>
    R3 Saffron:LinearColor=<rgba 0.2188 0.1677 0 1>
    R3 SkyBlue:LinearColor=<rgba 0 0.1213 0.1823 1>
    R3 TitaniumWhite:LinearColor=<rgba 0.1406 0.1406 0.1406 1>
    Use R1 Painted Overrides=true
    Use R2 Painted Overrides=true
    Use R3 Painted Overrides=true
   ) {module=VC_PaintedVariantsOverride001}
   &ParticleState()
   &InheritSourceMovement(
    Applied Position Scale=&Multiply_VectorByFloat(
     Vector=<1 1 1>
     Float=&FloatFromCurve(FloatCurve=DataInterface() {propertyValues {Curve {keys [{interpMode=RCIM_Cubic, tangentMode=RCTM_Break, time=0.5, value=1}, {interpMode=RCIM_Cubic, tangentMode=RCTM_Break, time=1, value=0.5}]}}})
    )
    ApplyForce=false
    ApplyPosition=true
    ApplyPositionFalloff=false
   ) {enabled=false}
   &SolveForcesAndVelocity(Clamp Velocity=false, Limit Acceleration=false)
   &VC_UpdateRibbon_v1(Enable Skeletal Position=true)
   /SetVariables(
    Particles.R1_Width:NiagaraFloat=&FloatFromCurve(
     FloatCurve=DataInterface() {propertyValues {Curve {keys [
         {interpMode=RCIM_Cubic, leaveTangent=2.6526, leaveTangentWeight=0.07, tangentMode=RCTM_Break, time=0.0019, value=0.3}
         {
          arriveTangent=-0.2453
          arriveTangentWeight=0.092
          interpMode=RCIM_Cubic
          leaveTangent=-0.2453
          leaveTangentWeight=0.1363
          tangentMode=RCTM_User
          time=0.35
          value=0.65
         }
         {
          arriveTangent=-1.178
          arriveTangentWeight=0.1489
          interpMode=RCIM_Cubic
          leaveTangent=-1.178
          leaveTangentWeight=0.2228
          tangentMode=RCTM_User
          time=0.7
          value=0.25
         }
         {arriveTangent=-0.5653, arriveTangentWeight=0.1553, interpMode=RCIM_Cubic, leaveTangent=-0.5653, tangentMode=RCTM_User, time=1, value=0.05}
        ]}}}
     Scale Curve=95
    )
    Particles.R2_Width:NiagaraFloat=&FloatFromCurve(
     FloatCurve=DataInterface() {propertyValues {Curve {keys [
         {interpMode=RCIM_Cubic, leaveTangentWeight=0.07, tangentMode=RCTM_Break, time=0, value=0.3}
         {
          arriveTangent=1.2202
          arriveTangentWeight=0.092
          interpMode=RCIM_Cubic
          leaveTangent=1.2202
          leaveTangentWeight=0.1363
          tangentMode=RCTM_User
          time=0.14
          value=1
         }
         {
          arriveTangent=-1.193
          arriveTangentWeight=0.1489
          interpMode=RCIM_Cubic
          leaveTangent=-1.193
          leaveTangentWeight=0.2228
          tangentMode=RCTM_User
          time=0.6257
          value=0.7288
         }
         {arriveTangentWeight=0.1553, interpMode=RCIM_Cubic, time=1, value=0}
        ]}}}
     Scale Curve=85
    )
   ) {isSetParameters=true}
   /SetVariables(
    Particles.R1_Width:NiagaraFloat=&FloatFromCurve(
     FloatCurve=DataInterface() {propertyValues {Curve {keys [
         {interpMode=RCIM_Cubic, leaveTangent=2.6526, leaveTangentWeight=0.07, tangentMode=RCTM_Break, time=0.0019, value=0.3}
         {
          arriveTangent=0.6138
          arriveTangentWeight=0.092
          interpMode=RCIM_Cubic
          leaveTangent=0.6138
          leaveTangentWeight=0.1363
          tangentMode=RCTM_User
          time=0.2
          value=0.7
         }
         {arriveTangent=-0.497, interpMode=RCIM_Cubic, leaveTangent=-0.497, tangentMode=RCTM_User, time=0.3995, value=0.6}
         {
          arriveTangent=-1.178
          arriveTangentWeight=0.1489
          interpMode=RCIM_Cubic
          leaveTangent=-1.178
          leaveTangentWeight=0.2228
          tangentMode=RCTM_User
          time=0.7
          value=0.35
         }
         {arriveTangentWeight=0.1553, interpMode=RCIM_Cubic, time=1, value=0.05}
        ]}}}
     Scale Curve=95
    )
    Particles.R2_Width:NiagaraFloat=&FloatFromCurve(
     FloatCurve=DataInterface() {propertyValues {Curve {keys [
         {interpMode=RCIM_Cubic, leaveTangentWeight=0.07, tangentMode=RCTM_Break, time=0, value=0.3}
         {
          arriveTangent=1.2202
          arriveTangentWeight=0.092
          interpMode=RCIM_Cubic
          leaveTangent=1.2202
          leaveTangentWeight=0.1363
          tangentMode=RCTM_User
          time=0.14
          value=1
         }
         {
          arriveTangent=-1.193
          arriveTangentWeight=0.1489
          interpMode=RCIM_Cubic
          leaveTangent=-1.193
          leaveTangentWeight=0.2228
          tangentMode=RCTM_User
          time=0.6257
          value=0.7288
         }
         {arriveTangentWeight=0.1553, interpMode=RCIM_Cubic, time=1, value=0}
        ]}}}
     Scale Curve=85
    )
   ) {enabled=false, isSetParameters=true}
   /SetVariables(Particles.DotProduct:NiagaraFloat=&DotProduct(
     Vector A=/Niagara/DynamicInputs/Math/TransformVectorByMatrix(
      Vector:Vector3f=<1 0 0>
      Transformation Matrix:NiagaraMatrix=$Engine.Owner.SystemLocalToWorld
      Transform Vector as Position:NiagaraBool=false
     )
     Vector B=&NormalizeVector(
      Vector To Normalize=&SubtractVectorFromPosition(
       Position=$Engine.Owner.Position
       Vector=&ReturnCameraPosition(Coordinate Space:ENiagaraCoordinateSpace=::Simulation, Camera Query:NiagaraDataInterfaceCamera=DataInterface())
      )
      Safe Normalize=true
      Safe Fallback Threshold=0
      Safe Fallback Vector=<0 0 0.001>
     )
    )) {isSetParameters=true}
   /SetVariables(Particles.CustomPosition_R1:NiagaraPosition=&AddVectorToPosition(
     Position=$Particles.Position
     Vector=&MakeVector(
      X=0
      Y=0
      Z=&MaskFloat(
       Mask Source=0
       Mask Parameter=?
       Mask=0
       Masked Float=&Cosine(
        Normalized Angle=&Add_Float(A=0, B=&Subtract_Float(A=$Emitter.NormalizedLoopAge, B=$Particles.NormalizedAge))
        Period=1
        Scale=&FloatFromCurve(
         FloatCurve=DataInterface() {propertyValues {Curve {keys [
             {arriveTangent=4.1101, interpMode=RCIM_Cubic, leaveTangent=4.1101, tangentMode=RCTM_User, time=0, value=0}
             {arriveTangent=3.5596, interpMode=RCIM_Cubic, leaveTangent=3.5596, time=0.4103, value=2}
             {
              arriveTangent=-0.5881
              arriveTangentWeight=0.5973
              interpMode=RCIM_Cubic
              leaveTangent=-0.5881
              leaveTangentWeight=0.5973
              tangentMode=RCTM_User
              time=0.7585
              value=2.7
             }
            ]}}}
         Scale Curve=&Lerp_Float(
          A=10
          B=6.25
          Alpha=&NormalizeFloat(Value=&Power(a=&Add_Float(A=$Particles.DotProduct, B=0.1), b=5), Min=0, Max=1)
         )
        )
        Bias=0
       )
       Passthrough Float=&Sine(
        Normalized Angle=&Add_Float(A=0, B=&Subtract_Float(A=$Emitter.NormalizedLoopAge, B=$Particles.NormalizedAge))
        Period=1
        Scale=&FloatFromCurve(
         FloatCurve=DataInterface() {propertyValues {Curve {keys [
             {arriveTangent=4.1101, interpMode=RCIM_Cubic, leaveTangent=4.1101, tangentMode=RCTM_User, time=0, value=0}
             {arriveTangent=3.5596, interpMode=RCIM_Cubic, leaveTangent=3.5596, time=0.4103, value=2}
             {
              arriveTangent=-0.5881
              arriveTangentWeight=0.5973
              interpMode=RCIM_Cubic
              leaveTangent=-0.5881
              leaveTangentWeight=0.5973
              tangentMode=RCTM_User
              time=0.7585
              value=2.7
             }
            ]}}}
         Scale Curve=&Lerp_Float(
          A=10
          B=6.25
          Alpha=&NormalizeFloat(Value=&Power(a=&Add_Float(A=$Particles.DotProduct, B=0.1), b=5), Min=0, Max=1)
         )
        )
        Bias=0
       )
      )
     )
    )) {isSetParameters=true}
  ]
  rendererClasses [&NiagaraRibbonRendererProperties]
  renderers [
   {rendererClass=&NiagaraRibbonRendererProperties}
   {rendererClass=&NiagaraRibbonRendererProperties, rendererIndex=1}
  ]
 }}
summary {
 systemName=NS_Soulfire
 emitters [{emitterName=NE_VC_Boost_Ribbon_Parent001, rendererClasses [&NiagaraRibbonRendererProperties]}]
 userVariables [
  User.bIsFrontendPreview {type=NiagaraBool, defaultValue=true}
  User.bIsPainted {type=NiagaraBool}
  User.ColorVariantID {type=NiagaraInt32}
  User.DTCarOffset002 {type=NiagaraPosition, description="", defaultValue {}}
  User.GameModeID {type=NiagaraInt32}
  User.PaintedColor {type=LinearColor, description="", defaultValue {struct=LinearColor, value=<rgba 1 1 1 1>}}
  User.PaintedScalar {type=NiagaraFloat, defaultValue=1, description=""}
  User.PortCount {type=NiagaraInt32, defaultValue=2}
 ]
}
